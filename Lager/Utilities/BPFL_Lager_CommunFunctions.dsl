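; Blueprint function library that starts, ends and checks Living World encounters for Lager activities.
; Start helpers validate inputs, get the GameState's Living World manager and pass it point providers from a volume, patrol paths or static providers.
; Any invalid provider aborts the start; only the volume variant validates the returned encounter, while list variants report true after starting.
; Elimination completes when no spawned actors are alive and the max spawn count is reached.
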(fn EndActivityEncounter
  (params (GameState :type Game%20State%20Base%20Object%20Reference) (Encounter :type Fort%20Athena%20Living%20World%20Encounter%20Instance%20Object%20Reference))
  (bind _output_get (LocalVariables|SetLivingWorldManager (Actor|GetComponentbyClass GameState "/Script/LagerRuntime.FortAthenaLivingWorldManager")))
  (Utilities|IsValid _output_get
    (:Is%20Valid
      (Class|FortAthenaLivingWorldManager|EndEncounter _output_get Encounter))))

(fn GetEncounterEliminationCompleted
  (params (LivingWorldEncounterInstance :type Fort%20Athena%20Living%20World%20Encounter%20Instance%20Object%20Reference))
  (if (and (<= (Class|FortAthenaLivingWorldEncounterInstance|GetAliveSpawnedActorCount LivingWorldEncounterInstance) 0) (Class|FortAthenaLivingWorldEncounterInstance|HasReachedMaxSpawnedCount LivingWorldEncounterInstance))
    (return true)
    (else
      (return false))))

(fn StartActivityEncounter
  (params (GetGameState :type Object%20Reference) (EncounterLivingWorldVolume :type Fort%20Athena%20Living%20World%20Volume%20Object%20Reference) (ActorDensityReservoir :type Actor%20Object%20Reference) (EncounterType :type Fort%20Athena%20Living%20World%20Encounter%20Soft%20Object%20Reference))
  (bind _returnvalue (Game|GetGameState GetGameState))
  (bind _returnvalue_2 (Actor|GetComponentbyClass _returnvalue "/Script/LagerRuntime.FortAthenaLivingWorldManager"))
  (Utilities|IsValid ActorDensityReservoir
    (:Is%20Valid
      (Utilities|IsValid _returnvalue
        (:Is%20Valid
          (Utilities|IsValid _returnvalue_2
            (:Is%20Valid
              (LocalVariables|SetLivingWorldManager _returnvalue_2)
              (Utilities|IsValid (Variables|GetEncounterLivingWorldVolume)
                (:Is%20Valid
                  (bind _asfort_athena_living_world_point_provider_interface (Utilities|Casting|CastToFortAthenaLivingWorldPointProviderInterface (Variables|GetEncounterLivingWorldVolume))
                    (:then
                      (bind _returnvalue_3 (Class|FortAthenaLivingWorldManager|StartEncounter (LocalVariables|GetLivingWorldManager) (Variables|GetEncounterType) (Utilities|Array|MakeArray _asfort_athena_living_world_point_provider_interface) (Variables|GetActorDensityReservoir) true))
                      (return (Utilities|IsValid _returnvalue_3 :!class /Script/Engine.KismetSystemLibrary) _returnvalue_3))
                    (:CastFailed
                      (return false 0))))
                (:Is%20Not%20Valid
                  (return false 0))))
            (:Is%20Not%20Valid
              (return false 0))))
        (:Is%20Not%20Valid
          (return false 0))))
    (:Is%20Not%20Valid
      (return false 0))))

(fn StartActivityEncounterPatrolPathPointProvider
  (params (GetGameState :type Object%20Reference) (ActorDensityReservoir :type Actor%20Object%20Reference) (EncounterType :type Fort%20Athena%20Living%20World%20Encounter%20Soft%20Object%20Reference) (LWPatrolPathPointProviders :type Array%20of%20Fort%20Athena%20Patrol%20Path%20Point%20Provider%20Object%20References))
  (bind _returnvalue (Game|GetGameState GetGameState))
  (bind _returnvalue_2 (Actor|GetComponentbyClass _returnvalue "/Script/LagerRuntime.FortAthenaLivingWorldManager"))
  (Utilities|IsValid ActorDensityReservoir
    (:Is%20Valid
      (Utilities|IsValid _returnvalue
        (:Is%20Valid
          (Utilities|IsValid _returnvalue_2
            (:Is%20Valid
              (LocalVariables|SetLivingWorldManager _returnvalue_2)
              (for _array_element (Variables|GetLWPatrolPathPointProviders)
                (Utilities|IsValid _array_element
                  (:Is%20Valid
                    (bind _asfort_athena_living_world_point_provider_interface (Utilities|Casting|CastToFortAthenaLivingWorldPointProviderInterface _array_element)
                      (:then
                        (Utilities|Array|Add (LocalVariables|GetEncounterPointProviders) _asfort_athena_living_world_point_provider_interface))
                      (:CastFailed
                        (return false 0))))
                  (:Is%20Not%20Valid
                    (return false 0))))
              (bind _returnvalue_3 (Class|FortAthenaLivingWorldManager|StartEncounter (LocalVariables|GetLivingWorldManager) (Variables|GetEncounterType) (LocalVariables|GetEncounterPointProviders) (Variables|GetActorDensityReservoir) true))
              (return true _returnvalue_3))
            (:Is%20Not%20Valid
              (return false 0))))
        (:Is%20Not%20Valid
          (return false 0))))
    (:Is%20Not%20Valid
      (return false 0))))

(fn StartActivityEncounterStaticPointProvider
  (params (GetGameState :type Object%20Reference) (ActorDensityReservoir :type Actor%20Object%20Reference) (EncounterType :type Fort%20Athena%20Living%20World%20Encounter%20Soft%20Object%20Reference) (LWStaticPointProviderList :type Array%20of%20Fort%20Athena%20Living%20World%20Static%20Point%20Provider%20Object%20References))
  (bind _returnvalue (Game|GetGameState GetGameState))
  (bind _returnvalue_2 (Actor|GetComponentbyClass _returnvalue "/Script/LagerRuntime.FortAthenaLivingWorldManager"))
  (Utilities|IsValid ActorDensityReservoir
    (:Is%20Valid
      (Utilities|IsValid _returnvalue
        (:Is%20Valid
          (Utilities|IsValid _returnvalue_2
            (:Is%20Valid
              (LocalVariables|SetLivingWorldManager _returnvalue_2)
              (for _array_element (Variables|GetLWStaticPointProviderList)
                (Utilities|IsValid _array_element
                  (:Is%20Valid
                    (bind _asfort_athena_living_world_point_provider_interface (Utilities|Casting|CastToFortAthenaLivingWorldPointProviderInterface _array_element)
                      (:then
                        (Utilities|Array|Add (LocalVariables|GetEncounterPointProviders) _asfort_athena_living_world_point_provider_interface))
                      (:CastFailed
                        (return false 0))))
                  (:Is%20Not%20Valid
                    (return false 0))))
              (bind _returnvalue_3 (Class|FortAthenaLivingWorldManager|StartEncounter (LocalVariables|GetLivingWorldManager) (Variables|GetEncounterType) (LocalVariables|GetEncounterPointProviders) (Variables|GetActorDensityReservoir) true))
              (return true _returnvalue_3))
            (:Is%20Not%20Valid
              (return false 0))))
        (:Is%20Not%20Valid
          (return false 0))))
    (:Is%20Not%20Valid
      (return false 0))))
